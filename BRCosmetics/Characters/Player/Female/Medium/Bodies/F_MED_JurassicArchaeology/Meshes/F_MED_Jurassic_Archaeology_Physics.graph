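# Physics asset for the Jurassic Archaeology female medium player body cosmetic, covering its dangling necklace and bandana.
# A three-link necklace chain and two bandana tails hang off small base bodies with angular limits, the necklace having zero twist range.
# Both bases attach to spine_05 with free rotation, and large spine_03 and spine_05 capsules provide torso shapes.
# A right pouch base body is present but has no constraint.
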PhysicsAsset /BRCosmetics/Characters/Player/Female/Medium/Bodies/F_MED_JurassicArchaeology/Meshes/F_MED_Jurassic_Archaeology_Physics
bodies {
 base_necklace [{shapeType=Capsule, length=0.505, radius=0.505, shapeName=base_necklace_capsule, rotation=<rot 0 -90 0>}]
 dyn_bandanaA_01 [{shapeType=Capsule, length=5.1617, radius=1.1413, center=<2.5333 0 0>, rotation=<rot 90 0 0>}]
 dyn_bandanaB_01 [{shapeType=Capsule, length=5.1617, radius=1.1413, center=<3.1104 0.0001 0>, rotation=<rot 90 0 0>}]
 dyn_bandanaBase [{shapeType=Capsule, length=0.505, radius=0.505, rotation=<rot 0 -90 0>}]
 dyn_necklace_01 [{
   shapeType=Capsule
   length=5.5031
   radius=1.0187
   shapeName=dyn_necklace_01_capsule
   center=<2.7562 -0.8202 -0.0095>
   rotation=<rot 89.4461 0.0001 0>
  }]
 dyn_necklace_02 [{
   shapeType=Capsule
   length=4.1945
   radius=0.9818
   shapeName=dyn_necklace_02_capsule
   center=<1.2066 -0.5225 0.0206>
   rotation=<rot 88.8438 -179.9999 180>
  }]
 dyn_necklace_03 [{
   shapeType=Capsule
   length=3.763
   radius=0.9528
   shapeName=dyn_necklace_03_capsule
   center=<2.9605 -0.6403 0.0449>
   rotation=<rot 88.8363 -180 -180>
  }]
 dyn_pouchB_Base_r [{shapeType=Capsule, length=0.505, radius=0.505, rotation=<rot 0 -90 0>}]
 spine_03 [{shapeType=Capsule, length=7.2564, radius=10.482, shapeName=spine_03_capsule, center=<7.1512 2.0768 0.0001>}]
 spine_05 [{shapeType=Capsule, length=10.5558, radius=10.7453, shapeName=spine_05_capsule, center=<-0.1711 -2.4848 0>}]
}
constraints [
 {bone1Name=dyn_bandanaA_01, bone2Name=dyn_bandanaBase, swing1LimitDegrees=20, swing2LimitDegrees=20, twistLimitDegrees=20}
 {bone1Name=dyn_bandanaB_01, bone2Name=dyn_bandanaBase, swing1LimitDegrees=20, swing2LimitDegrees=20, twistLimitDegrees=20}
 {bone1Name=dyn_necklace_02, bone2Name=dyn_necklace_01, swing1LimitDegrees=20, swing2LimitDegrees=40, twistLimitDegrees=0}
 {bone1Name=dyn_necklace_03, bone2Name=dyn_necklace_02, swing1LimitDegrees=20, swing2LimitDegrees=20, twistLimitDegrees=0}
 {
  bone1Name=dyn_bandanaBase
  bone2Name=spine_05
  swing1LimitDegrees=45
  swing1Motion=Free
  swing2LimitDegrees=45
  swing2Motion=Free
  twistLimitDegrees=45
  twistMotion=Free
 }
 {bone1Name=dyn_necklace_01, bone2Name=base_necklace, swing1LimitDegrees=20, swing2LimitDegrees=60, twistLimitDegrees=0}
 {
  bone1Name=base_necklace
  bone2Name=spine_05
  swing1LimitDegrees=45
  swing1Motion=Free
  swing2LimitDegrees=45
  swing2Motion=Free
  twistLimitDegrees=45
  twistMotion=Free
 }
]
