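# Animation Blueprint for the female medium Jurassic Archaeology cosmetic body, adding secondary motion and pouch clipping fixes.
# It copies the attached parent's pose through a deform Control Rig, constrains the bandana base, and simulates RigidBody and AnimDynamics dangles.
# Additive bone offsets then shift pouches through interpolated alphas.
# The update event derives those alphas from jumping, movement direction and thigh pitch during moving crouch or knee slide.
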AnimBlueprint /BRCosmetics/Characters/Player/Female/Medium/Bodies/F_MED_JurassicArchaeology/Meshes/F_MED_Jurassic_Archaeology_AnimBP
target_skeleton=&/F_MED_Jurassic_Archaeology_Skeleton
graphs [
 AnimGraph {nodes [
   n1 = Anim Dynamics - Bone: dyn_largePouch_l(ComponentPose=Local To Component(LocalPose=Use cached pose 'Pouch fix Pose'()), Alpha=1, ExternalForce="0,0,0") {
    angularDampingOverride=1
    angularSpringConstant=20000
    bAngularSpring=true
    bOverrideAngularDamping=true
    bUseGravityOverride=true
    gravityScale=0.35
    simulationSpace=BoneRelative
    alphaBoolBlend {blendOption=Linear}
    boundBone {boneName=dyn_largePouch_l}
    componentLinearAccScale=<1 1 1>
    componentLinearVelScale=<1 1 1>
    gravityOverride {z=-980}
    physicsBodyDefinitions [{
      boundBone {boneName=dyn_largePouch_l}
      boxExtents=<20 30 30>
      constraintSetup {angularLimitsMax=<1 0 0>, angularLimitsMin=<-1 0 -8>, angularTarget=<1 0 0>}
      localJointOffset=<2 0 0>
     }]
    relativeSpaceBone {boneName=pelvis}
   }
   n2 = Anim Dynamics - Bone: dyn_pouchA_r(ComponentPose=@n1, Alpha=1, ExternalForce="0,0,0") {
    angularDampingOverride=1
    angularSpringConstant=20000
    bAngularSpring=true
    bOverrideAngularDamping=true
    bUseGravityOverride=true
    gravityScale=0.35
    simulationSpace=BoneRelative
    alphaBoolBlend {blendOption=Linear}
    boundBone {boneName=dyn_pouchA_r}
    componentLinearAccScale=<1 1 1>
    componentLinearVelScale=<1 1 1>
    gravityOverride {z=-980}
    physicsBodyDefinitions [{
      boundBone {boneName=dyn_pouchA_r}
      boxExtents=<20 30 30>
      constraintSetup {angularLimitsMax=<1 0 5>, angularLimitsMin=<-1 0 -8>, angularTarget=<1 0 0>}
      localJointOffset=<2 0 0>
     }]
    relativeSpaceBone {boneName=pelvis}
   }
   n3 = Anim Dynamics - Bone: dyn_pouchB_r(ComponentPose=@n2, Alpha=1, ExternalForce="0,0,0") {
    angularDampingOverride=1
    angularSpringConstant=20000
    bAngularSpring=true
    bOverrideAngularDamping=true
    bUseGravityOverride=true
    gravityScale=0.35
    simulationSpace=BoneRelative
    alphaBoolBlend {blendOption=Linear}
    boundBone {boneName=dyn_pouchB_r}
    componentLinearAccScale=<1 1 1>
    componentLinearVelScale=<1 1 1>
    gravityOverride {z=-980}
    physicsBodyDefinitions [{
      boundBone {boneName=dyn_pouchB_r}
      boxExtents=<20 30 30>
      constraintSetup {angularLimitsMax=<1 0 5>, angularLimitsMin=<-1 0 -8>, angularTarget=<1 0 0>}
      localJointOffset=<2 0 0>
     }]
    relativeSpaceBone {boneName=pelvis}
   }
   n4 = "Transform (Modify) Bone - Bone: dyn_Base_l"(ComponentPose=@n3, Alpha=1, Translation="0.083,-0.926,0.278", Rotation="0,-10,0.001", Scale="1,1,1") {
    rotationMode=BMM_Additive
    rotationSpace=BCS_BoneSpace
    translationMode=BMM_Additive
    translationSpace=BCS_BoneSpace
    alphaBoolBlend {blendInTime=0.5, blendOption=Linear, blendOutTime=0.5}
    boneToModify {boneName=dyn_Base_l}
    rotation {roll=0.001, yaw=-10}
    translation=<0.083 -0.926 0.278>
   }
   n5 = "Transform (Modify) Bone - Bone: dyn_pouchA_Base_r"(
    ComponentPose=@n4
    Alpha=Get isJumping() -> (isJumping:double)
    Translation="0.4507,-5.2823,-2.0827"
    Rotation="1.7917,-18.8708,23.9902"
    Scale="1,1,1"
   ) {
    rotationMode=BMM_Additive
    rotationSpace=BCS_BoneSpace
    translationMode=BMM_Additive
    translationSpace=BCS_BoneSpace
    alphaBoolBlend {blendOption=Linear}
    alphaScaleBiasClamp {bInterpResult=true, interpSpeedDecreasing=1.5, interpSpeedIncreasing=2}
    boneToModify {boneName=dyn_pouchA_Base_r}
    rotation=<rot 1.7917 -18.8708 23.9902>
    translation=<0.4507 -5.2823 -2.0827>
   }
   n6 = "Transform (Modify) Bone - Bone: dyn_Base_l"(
    ComponentPose=@n5
    Alpha=Get isStrafe_L() -> (isStrafe_L:double)
    Translation="-1.151,-2.013,-1.532"
    Rotation="2.969,-11.604,-16.890"
    Scale="1,1,1"
   ) {
    rotationMode=BMM_Additive
    rotationSpace=BCS_BoneSpace
    translationMode=BMM_Additive
    translationSpace=BCS_BoneSpace
    alphaBoolBlend {blendOption=Linear}
    alphaScaleBiasClamp {bInterpResult=true, interpSpeedDecreasing=8, interpSpeedIncreasing=8}
    boneToModify {boneName=dyn_Base_l}
    rotation=<rot 2.9691 -11.604 -16.8901>
    translation=<-1.151 -2.013 -1.532>
   }
   n7 = Get isStrafe_R() -> (isStrafe_R:double)
   n8 = "Transform (Modify) Bone - Bone: dyn_pouchA_Base_r"(ComponentPose=@n6, Alpha=@n7, Translation="-0.699,-3.942,-1.081", Rotation="-8.313,-33.692,25.501", Scale="1,1,1") {
    rotationMode=BMM_Additive
    rotationSpace=BCS_BoneSpace
    translationMode=BMM_Additive
    translationSpace=BCS_BoneSpace
    alphaBoolBlend {blendOption=Linear}
    alphaScaleBiasClamp {bInterpResult=true, interpSpeedDecreasing=12, interpSpeedIncreasing=14}
    boneToModify {boneName=dyn_pouchA_Base_r}
    rotation=<rot -8.313 -33.692 25.501>
    translation=<-0.699 -3.942 -1.081>
   }
   n9 = "Transform (Modify) Bone - Bone: dyn_pouchB_Base_r"(ComponentPose=@n8, Alpha=@n7, Translation="0.078,-0.344,-0.706", Rotation="-5.737,-11.058,21.307", Scale="1,1,1") {
    rotationMode=BMM_Additive
    rotationSpace=BCS_BoneSpace
    translationMode=BMM_Additive
    translationSpace=BCS_BoneSpace
    alphaBoolBlend {blendOption=Linear}
    alphaScaleBiasClamp {bInterpResult=true, interpSpeedDecreasing=5}
    boneToModify {boneName=dyn_pouchB_Base_r}
    rotation=<rot -5.7371 -11.0579 21.3071>
    translation=<0.078 -0.344 -0.706>
   }
   n10 = Get isStrafe_R_opposite() -> (isStrafe_R_opposite:double)
   n11 = "Transform (Modify) Bone - Bone: dyn_pouchA_Base_r"(ComponentPose=@n9, Alpha=@n10, Translation="-0.537,-1.480,0.478", Rotation="1.630,-19.443,16.802", Scale="1,1,1") {
    rotationMode=BMM_Additive
    rotationSpace=BCS_BoneSpace
    translationMode=BMM_Additive
    translationSpace=BCS_BoneSpace
    alphaBoolBlend {blendOption=Linear}
    alphaScaleBiasClamp {bInterpResult=true, interpSpeedIncreasing=14}
    boneToModify {boneName=dyn_pouchA_Base_r}
    rotation=<rot 1.63 -19.443 16.802>
    translation=<-0.537 -1.4801 0.478>
   }
   n12 = "Transform (Modify) Bone - Bone: dyn_pouchB_Base_r"(ComponentPose=@n11, Alpha=@n10, Translation="0,0,0", Rotation="-0.000,-14.060,0.000", Scale="1,1,1") {
    rotationMode=BMM_Additive
    rotationSpace=BCS_BoneSpace
    translationMode=BMM_Additive
    translationSpace=BCS_BoneSpace
    alphaBoolBlend {blendOption=Linear}
    alphaScaleBiasClamp {bInterpResult=true, interpSpeedIncreasing=13.5}
    boneToModify {boneName=dyn_pouchB_Base_r}
    rotation=<rot -0.0001 -14.06 0.0001>
   }
   n13 = "Transform (Modify) Bone - Bone: dyn_pouchA_Base_r"(
    ComponentPose=@n12
    Alpha=Get Crouch_clearance_r() -> (Crouch_clearance_r:double)
    Translation="-0.000,-1.791,-1.131"
    Rotation="-0.000,0.000,14.065"
    Scale="1,1,1"
   ) {
    rotationMode=BMM_Additive
    rotationSpace=BCS_BoneSpace
    translationMode=BMM_Additive
    translationSpace=BCS_BoneSpace
    alphaBoolBlend {blendOption=Linear}
    alphaScaleBiasClamp {bInterpResult=true, interpSpeedIncreasing=30}
    boneToModify {boneName=dyn_pouchA_Base_r}
    rotation=<rot 0 0 14.065>
    translation {y=-1.791, z=-1.131}
   }
   Output Pose(Result=Component To Local(ComponentPose=@n13)) {layerGroup=None}
   n14 = Control Rig(Source=Copy Pose From Mesh() {bUseAttachedParent=true}) {
    alphaBoolBlend {blendOption=Linear}
    controlRigAssetReference {blueprintRigClass=&/F_MED_DeformRig_Jurassic_Archaeology.F_MED_DeformRig_Jurassic_Archaeology_C}
   }
   n15 = "Constraint - dyn_bandanaBase (PPP)"!(
    ComponentPose=Local To Component(LocalPose=@n14)
    Alpha=1
    ConstraintWeights_0=0.1175
    ConstraintWeights_1:float=0.7521
    ConstraintWeights_2:float=0.1304
   ) -> (Pose) {
    alphaBoolBlend {blendOption=Linear}
    boneToModify {boneName=dyn_bandanaBase}
    constraintSetup [{targetBone {boneName=upperarm_r}}, {targetBone {boneName=upperarm_twist_01_r}}, {targetBone {boneName=upperarm_twist_02_r}}]
    constraintWeights [1,1,1]
   }
   n16 = RigidBody!(ComponentPose=@n15, Alpha, OverrideWorldGravity:Vector="0,0,-980", ExternalForce) -> (Pose) {
    bClampLinearTranslationLimitToRefPose=true
    bOverrideWorldGravity=true
    overridePhysicsAsset=&/F_MED_Jurassic_Archaeology_Physics
    simulationSpace=BaseBoneSpace
    alphaBoolBlend {blendOption=Linear}
    baseBoneRef {boneName=spine_05}
    componentAppliedLinearAccClamp=<2000 2000 1000>
    componentLinearAccScale {x=0.4}
    componentLinearVelScale {x=0.4}
    overrideWorldGravity {z=-980}
    simSpaceSettings {maxAngularAcceleration=100, maxAngularVelocity=5, maxLinearAcceleration=300, maxLinearVelocity=50, velocityScaleZ=0, worldAlpha=0.5}
   }
   n17 = Anim Dynamics - Bone: dyn_laces_r(ComponentPose=@n16, Alpha=1, ExternalForce="0,0,0") {
    angularDampingOverride=1
    angularSpringConstant=20000
    bAngularSpring=true
    bOverrideAngularDamping=true
    bUseGravityOverride=true
    gravityScale=0.35
    simulationSpace=RootRelative
    alphaBoolBlend {blendOption=Linear}
    boundBone {boneName=dyn_laces_r}
    componentLinearAccScale=<1 1 1>
    componentLinearVelScale=<1 1 1>
    gravityOverride {z=-980}
    physicsBodyDefinitions [{
      boundBone {boneName=dyn_laces_r}
      boxExtents=<20 30 30>
      constraintSetup {angularLimitsMax=<15 15 15>, angularLimitsMin=<-15 -15 -15>, angularTarget=<1 0 0>}
      localJointOffset=<4 0 0>
     }]
   }
   n18 = Anim Dynamics - Bone: dyn_laces_l(ComponentPose=@n17, Alpha=1, ExternalForce="0,0,0") {
    angularDampingOverride=1
    angularSpringConstant=20000
    bAngularSpring=true
    bOverrideAngularDamping=true
    bUseGravityOverride=true
    gravityScale=0.35
    simulationSpace=RootRelative
    alphaBoolBlend {blendOption=Linear}
    boundBone {boneName=dyn_laces_l}
    componentLinearAccScale=<1 1 1>
    componentLinearVelScale=<1 1 1>
    gravityOverride {z=-980}
    physicsBodyDefinitions [{
      boundBone {boneName=dyn_laces_l}
      boxExtents=<20 30 30>
      constraintSetup {angularLimitsMax=<15 15 15>, angularLimitsMin=<-15 -15 -15>, angularTarget=<1 0 0>}
      localJointOffset=<4 0 0>
     }]
   }
   n19 = "Transform (Modify) Bone - Bone: belt_helper"(ComponentPose=@n18, Alpha=1, Translation="0.673,0.899,0.001", Rotation="0.008,-0.024,-1.977", Scale="1,1,1") {
    rotationMode=BMM_Additive
    rotationSpace=BCS_BoneSpace
    translationMode=BMM_Additive
    translationSpace=BCS_BoneSpace
    alphaBoolBlend {blendOption=Linear}
    alphaScaleBiasClamp {bInterpResult=true}
    boneToModify {boneName=belt_helper}
    rotation=<rot 0.008 -0.024 -1.977>
    translation=<0.673 0.899 0.001>
   }
   Save cached pose 'Pouch fix Pose'(Pose=Component To Local(ComponentPose=@n19)) {cacheName=Pouch fix Pose}
  ]}
 EventGraph {nodes [
   Event BlueprintUpdateAnimation()
   > n1 = Sequence!(execute) -> (then_0, then_1, then_2:Exec, then_3:Exec, then_4:Exec)
   > Set crouchAlpha(crouchAlpha:bool=Get bIsCrouching()) -> (Output_Get:bool)
   n2 = Get bIsJumping()
   n3 = Get bIsMoving2D()
   > Set isJumping(isJumping:double="To Float (Boolean)"(InBool=AND Boolean(A=@n3, B=@n2))) -> (Output_Get:double)
   n4 = PlayerMovement(execute=@n1.then_3, AnimInstance=Self-Reference(), XDotProductCoefficient=100, YDotProductCoefficient=100, ZDotProductCoefficient=100)
   n5 = BreakVector(InVec=MakeVector(X=@n4, Y=@n4.YMovement, Z=@n4.ZMovement))
   > Set Left(Left:bool="float < float"(A=@n5, B=-50)) -> (Output_Get:bool)
   > Set Right(Right:bool="float > float"(A=@n5, B=50)) -> (Output_Get:bool)
   > Set Forward(Forward:bool="float > float"(A=@n5.Y, B=50)) -> (Output_Get:bool)
   > Set Backward(Backward:bool="float < float"(A=@n5.Y, B=-50)) -> (Output_Get:bool)
   n6 = Get Left() -> (Left:bool)
   n7 = Get Left_leg_rot() -> (Left_leg_rot:bool)
   > Set isStrafe_L(isStrafe_L:double="To Float (Boolean)"(InBool=AND Boolean(A=@n7, B=@n6))) -> (Output_Get:double)
   n8 = Get Right() -> (Right:bool)
   n9 = Get Left_leg_rot()
   > Set isStrafe_L_opposite(isStrafe_L_opposite:double="To Float (Boolean)"(InBool=AND Boolean(A=@n9, B=@n8))) -> (Output_Get:double)
   n10 = Get Right()
   n11 = Get Right_leg_rot() -> (Right_leg_rot:bool)
   > Set isStrafe_R(isStrafe_R:double="To Float (Boolean)"(InBool=AND Boolean(A=@n11, B=@n10))) -> (Output_Get:double)
   n12 = Get Left()
   n13 = Get Right_leg_rot_Special() -> (Right_leg_rot_Special:bool)
   > Set isStrafe_R_opposite(isStrafe_R_opposite:double="To Float (Boolean)"(InBool=AND Boolean(A=@n13, B=@n12))) -> (Output_Get:double)
   n14 = OR Boolean(A=Get Forward() -> (Forward:bool), B=Get Backward() -> (Backward:bool))
   n15 = Get Left_leg_rot()
   > Set Crouch_clearance_l(Crouch_clearance_l:double="To Float (Boolean)"(InBool=AND Boolean(A=@n14, B=@n15))) -> (Output_Get:double)
   n16 = OR Boolean(A=Get Forward(), B=Get Backward())
   n17 = Get Right_leg_rot_Special()
   > Set Crouch_clearance_r(Crouch_clearance_r:double="To Float (Boolean)"(InBool=AND Boolean(A=@n16, B=@n17))) -> (Output_Get:double)
   n18 = AND Boolean(A=Get bIsCrouching(), B=Get bIsMoving2D())
   n19 = Get bIsKneeSliding()
   n20 = GetOwnerSkeletalMeshComponent()
   n21 = BreakRotator(InRot=BreakTransform(InTransform=GetDeltaTransformFromRefPose(self=@n20, BoneName=thigh_r, BaseName=pelvis)).Rotation)
   Set Right_leg_rot_Special(execute=@n1.then_4, Right_leg_rot_Special:bool=AND Boolean(A="float >= float"(A=@n21.Pitch, B=2), B=OR Boolean(A=@n18, B=@n19))) -> (Output_Get:bool)
   n22 = AND Boolean(A=Get bIsCrouching(), B=Get bIsMoving2D())
   n23 = Get bIsKneeSliding()
   n24 = GetOwnerSkeletalMeshComponent()
   n25 = BreakRotator(InRot=BreakTransform(InTransform=GetDeltaTransformFromRefPose(self=@n24, BoneName=thigh_l, BaseName=pelvis)).Rotation)
   Set Left_leg_rot(execute=@n1.then_1, Left_leg_rot:bool=AND Boolean(A="float >= float"(A=@n25.Pitch, B=2), B=OR Boolean(A=@n22, B=@n23))) -> (Output_Get:bool)
   n26 = AND Boolean(A=Get bIsCrouching(), B=Get bIsMoving2D())
   n27 = Get bIsKneeSliding()
   n28 = GetOwnerSkeletalMeshComponent()
   n29 = BreakRotator(InRot=BreakTransform(InTransform=GetDeltaTransformFromRefPose(self=@n28, BoneName=thigh_r, BaseName=pelvis)).Rotation)
   Set Right_leg_rot(execute=@n1.then_2, Right_leg_rot:bool=AND Boolean(A="float >= float"(A=@n29.Pitch, B=17), B=OR Boolean(A=@n26, B=@n27))) -> (Output_Get:bool)
   TryGetPawnOwner()
  ]}
]
variables [
 Backward {type=boolean}
 Crouch_clearance_l {type=number}
 Crouch_clearance_r {type=number}
 crouchAlpha {type=boolean}
 Forward {type=boolean}
 isJumping {type=number}
 isStrafe_L {type=number}
 isStrafe_L_opposite {type=number}
 isStrafe_R {type=number}
 isStrafe_R_opposite {type=number}
 Left {type=boolean}
 Left_leg_rot {type=boolean}
 Leg bend {type=number}
 Right {type=boolean}
 Right_leg_rot {type=boolean}
 Right_leg_rot_Special {type=boolean}
]
