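# Physics asset for the F_MED Jurassic Archaeology face accessory mesh, defining bodies and joints for its pigtails and temple hair.
# Small base capsules join each hair chain to the head with free rotation, followed by capsules with tight swing and twist limits.
# Pigtail B and C strands branch from the third segment of each main pigtail.
# Head, neck, upper spine, clavicle and upper-arm shapes form a free-jointed upper body, apparently as hair colliders.
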PhysicsAsset /BRCosmetics/Characters/Player/Female/Medium/Bodies/F_MED_JurassicArchaeology/Meshes/Parts/F_MED_Jurassic_Archaeology_FaceAcc_Physics
bodies {
 clavicle_l [{shapeType=Capsule, length=13.7379, radius=6, center=<6.3153 -1.2226 -0.95>, rotation=<rot 90 0 0>}]
 clavicle_r [{shapeType=Capsule, length=13.7379, radius=6, center=<-6.3406 0.6526 2.016>, rotation=<rot 90 0 0>}]
 dyn_pigtailA_01_Base_l [{shapeType=Capsule, length=0.505, radius=0.505, rotation=<rot 90 0 0>}]
 dyn_pigtailA_01_Base_r [{shapeType=Capsule, length=0.505, radius=0.505, rotation=<rot 90 0 0>}]
 dyn_pigtailA_01_l [{shapeType=Capsule, length=3.9706, radius=3.6158, center=<2.7741 0 0>, rotation=<rot 90 0 -179.9999>}]
 dyn_pigtailA_01_r [{shapeType=Capsule, length=3.9706, radius=3.6158, center=<2.7235 0 0>, rotation=<rot 90 0 0>}]
 dyn_pigtailA_02_l [{shapeType=Capsule, length=4.9505, radius=2.5691, center=<2.9418 0 0>, rotation=<rot 90 0 -179.9999>}]
 dyn_pigtailA_02_r [{shapeType=Capsule, length=4.9505, radius=2.5691, center=<3.5782 0 0>, rotation=<rot 90 0 0>}]
 dyn_pigtailA_03_l [{shapeType=Capsule, length=4.3316, radius=1.8714, center=<2.9183 0 0>, rotation=<rot 90 0 -179.9999>}]
 dyn_pigtailA_03_r [{shapeType=Capsule, length=4.3316, radius=1.8714, center=<2.0889 0 0>, rotation=<rot 90 0 0>}]
 dyn_pigtailA_04_l [{shapeType=Capsule, length=3.492, radius=1.1733, center=<1.4617 0 0>, rotation=<rot 90 0 -179.9999>}]
 dyn_pigtailA_04_r [{shapeType=Capsule, length=3.492, radius=1.1733, center=<1.445 0 0>, rotation=<rot 90 0 0>}]
 dyn_pigtailA_05_l [{shapeType=Capsule, length=2.1235, radius=1.4154, center=<1.2474 0 0>, rotation=<rot 90 0 -179.9999>}]
 dyn_pigtailA_05_r [{shapeType=Capsule, length=2.1235, radius=1.4154, center=<1.7238 0 0>, rotation=<rot 90 0 0>}]
 dyn_pigtailB_01 [{shapeType=Capsule, length=2.4189, radius=0.6846, center=<1.485 0 0>, rotation=<rot 90 0 0>}]
 dyn_pigtailB_01_l [{shapeType=Capsule, length=0.8637, radius=0.969, center=<1.0108 0 0>, rotation=<rot 90 0 0>}]
 dyn_pigtailB_02 [{shapeType=Capsule, length=1.7033, radius=0.5836, center=<0.6851 0 0>, rotation=<rot 90 0 12.3536>}]
 dyn_pigtailB_02_l [{shapeType=Capsule, length=2.5904, radius=0.6303, center=<1.5105 0 0>, rotation=<rot 90 0 0>}]
 dyn_pigtailC_01 [{shapeType=Capsule, length=1.626, radius=0.9615, center=<1.2715 0 0>, rotation=<rot 90 0 0>}]
 dyn_pigtailC_01_l [{shapeType=Capsule, length=2.8367, radius=0.8043, center=<1.5323 0 0>, rotation=<rot 90 0 0>}]
 dyn_pigtailC_02 [{shapeType=Capsule, length=3.2312, radius=0.9373, center=<2.1819 0 0>, rotation=<rot 82.7333 0.0002 0.0002>}]
 dyn_templeHairA_01 [{shapeType=Capsule, length=3.8935, radius=1.0457, center=<2.0935 0 0>, rotation=<rot 90 0 0>}]
 dyn_templeHairA_01_Base [{shapeType=Capsule, length=0.505, radius=0.505, rotation=<rot 90 0 0>}]
 dyn_templeHairB_01 [{shapeType=Capsule, length=3.0267, radius=0.6626, center=<1.6062 0 0>, rotation=<rot 90 0 0>}]
 dyn_templeHairB_01_Base [{shapeType=Capsule, length=0.505, radius=0.505, rotation=<rot 90 0 0>}]
 dyn_templeHairC_01 [{shapeType=Capsule, length=2.8191, radius=1.7279, center=<1.0634 0 -0.1546>, rotation=<rot 82.0376 0 0>}]
 dyn_templeHairC_01_Base [{shapeType=Capsule, length=0.505, radius=0.505, rotation=<rot 90 0 0>}]
 dyn_templeHair_01_Base_r [{shapeType=Capsule, length=0.505, radius=0.505, rotation=<rot 90 0 0>}]
 dyn_templeHair_01_r [{shapeType=Capsule, length=1.4052, radius=1.7338, center=<1.4227 0 0>, rotation=<rot 90 0 0>}]
 dyn_templeHair_02_r [{shapeType=Capsule, length=2.0079, radius=1.2337, center=<2.0791 0 0>, rotation=<rot 90 0 0>}]
 dyn_templeHair_03_r [{shapeType=Capsule, length=2.4431, radius=0.711, center=<1.7757 0.0001 0>, rotation=<rot 90 0 0>}]
 dyn_templeHair_04_r [{shapeType=Capsule, length=2.0359, radius=0.7103, center=<1.4945 0 0>, rotation=<rot 90 0 0>}]
 dyn_templeHair_05_r [{shapeType=Capsule, length=0.505, radius=0.505, rotation=<rot 90 0 0>}]
 head [
  {shapeType=Sphere, radius=7.9918, center=<6.569 -1.8592 0>}
  {shapeType=Capsule, length=14.2689, radius=5.82, center=<3.4041 0 0>, rotation=<rot 90 0 0>}
 ]
 neck_01 [{shapeType=Capsule, length=4, radius=9, shapeName=neck_01_capsule, rotation=<rot 90 0 0>}]
 spine_05 [{shapeType=Capsule, length=18.0864, radius=9, center=<1.2588 -0.6141 0>, rotation=<rot 0 -90 0>}]
 upperarm_l [{shapeType=Capsule, length=25, radius=5, shapeName=upperarm_l_capsule, center=<15 0 0>, rotation=<rot 90 0 0>}]
 upperarm_r [{shapeType=Capsule, length=25, radius=5, shapeName=upperarm_r_capsule, center=<-15 0 0>, rotation=<rot 90 0 0>}]
}
constraints [
 {bone1Name=dyn_pigtailC_01_l, bone2Name=dyn_pigtailA_03_l, swing1LimitDegrees=15, swing2LimitDegrees=15, twistLimitDegrees=10}
 {
  bone1Name=dyn_templeHair_01_Base_r
  bone2Name=head
  swing1LimitDegrees=45
  swing1Motion=Free
  swing2LimitDegrees=45
  swing2Motion=Free
  twistLimitDegrees=45
  twistMotion=Free
 }
 {bone1Name=dyn_templeHair_01_r, bone2Name=dyn_templeHair_01_Base_r, swing1LimitDegrees=16, swing2LimitDegrees=16, twistLimitDegrees=12}
 {bone1Name=dyn_templeHair_02_r, bone2Name=dyn_templeHair_01_r, swing1LimitDegrees=16, swing2LimitDegrees=16, twistLimitDegrees=12}
 {bone1Name=dyn_templeHair_03_r, bone2Name=dyn_templeHair_02_r, swing1LimitDegrees=16, swing2LimitDegrees=16, twistLimitDegrees=12}
 {bone1Name=dyn_templeHair_04_r, bone2Name=dyn_templeHair_03_r, swing1LimitDegrees=16, swing2LimitDegrees=16, twistLimitDegrees=12}
 {bone1Name=dyn_templeHair_05_r, bone2Name=dyn_templeHair_04_r, swing1LimitDegrees=20, swing2LimitDegrees=20, twistLimitDegrees=20}
 {
  bone1Name=dyn_templeHairA_01_Base
  bone2Name=head
  swing1LimitDegrees=45
  swing1Motion=Free
  swing2LimitDegrees=45
  swing2Motion=Free
  twistLimitDegrees=45
  twistMotion=Free
 }
 {bone1Name=dyn_templeHairA_01, bone2Name=dyn_templeHairA_01_Base, swing1LimitDegrees=20, swing2LimitDegrees=20, twistLimitDegrees=12}
 {
  bone1Name=dyn_templeHairB_01_Base
  bone2Name=head
  swing1LimitDegrees=45
  swing1Motion=Free
  swing2LimitDegrees=45
  swing2Motion=Free
  twistLimitDegrees=45
  twistMotion=Free
 }
 {bone1Name=dyn_templeHairB_01, bone2Name=dyn_templeHairB_01_Base, swing1LimitDegrees=20, swing2LimitDegrees=20, twistLimitDegrees=12}
 {
  bone1Name=dyn_templeHairC_01_Base
  bone2Name=head
  swing1LimitDegrees=45
  swing1Motion=Free
  swing2LimitDegrees=45
  swing2Motion=Free
  twistLimitDegrees=45
  twistMotion=Free
 }
 {bone1Name=dyn_templeHairC_01, bone2Name=dyn_templeHairC_01_Base, swing1LimitDegrees=20, swing2LimitDegrees=20, twistLimitDegrees=12}
 {
  bone1Name=dyn_pigtailA_01_Base_r
  bone2Name=head
  swing1LimitDegrees=45
  swing1Motion=Free
  swing2LimitDegrees=45
  swing2Motion=Free
  twistLimitDegrees=45
  twistMotion=Free
 }
 {bone1Name=dyn_pigtailA_01_r, bone2Name=dyn_pigtailA_01_Base_r, swing1LimitDegrees=22, swing2LimitDegrees=22, twistLimitDegrees=12}
 {bone1Name=dyn_pigtailA_02_r, bone2Name=dyn_pigtailA_01_r, swing1LimitDegrees=22, swing2LimitDegrees=22, twistLimitDegrees=12}
 {bone1Name=dyn_pigtailA_03_r, bone2Name=dyn_pigtailA_02_r, swing1LimitDegrees=32, swing2LimitDegrees=32, twistLimitDegrees=12}
 {bone1Name=dyn_pigtailA_04_r, bone2Name=dyn_pigtailA_03_r, swing1LimitDegrees=32, swing2LimitDegrees=32, twistLimitDegrees=12}
 {bone1Name=dyn_pigtailA_05_r, bone2Name=dyn_pigtailA_04_r, swing1LimitDegrees=32, swing2LimitDegrees=32, twistLimitDegrees=12}
 {bone1Name=dyn_pigtailC_01, bone2Name=dyn_pigtailA_03_r, swing1LimitDegrees=12, swing2LimitDegrees=12, twistLimitDegrees=8}
 {bone1Name=dyn_pigtailC_02, bone2Name=dyn_pigtailC_01, swing1LimitDegrees=12, swing2LimitDegrees=12, twistLimitDegrees=8}
 {bone1Name=dyn_pigtailB_01, bone2Name=dyn_pigtailA_03_r, swing1LimitDegrees=12, swing2LimitDegrees=12, twistLimitDegrees=8}
 {bone1Name=dyn_pigtailB_02, bone2Name=dyn_pigtailB_01, swing1LimitDegrees=12, swing2LimitDegrees=12, twistLimitDegrees=8}
 {
  bone1Name=dyn_pigtailA_01_Base_l
  bone2Name=head
  swing1LimitDegrees=45
  swing1Motion=Free
  swing2LimitDegrees=45
  swing2Motion=Free
  twistLimitDegrees=45
  twistMotion=Free
 }
 {bone1Name=dyn_pigtailB_01_l, bone2Name=dyn_pigtailA_03_l, swing1LimitDegrees=15, swing2LimitDegrees=15, twistLimitDegrees=10}
 {bone1Name=dyn_pigtailB_02_l, bone2Name=dyn_pigtailB_01_l, swing1LimitDegrees=15, swing2LimitDegrees=15, twistLimitDegrees=10}
 {bone1Name=dyn_pigtailA_01_l, bone2Name=dyn_pigtailA_01_Base_l, swing1LimitDegrees=22, swing2LimitDegrees=22, twistLimitDegrees=12}
 {bone1Name=dyn_pigtailA_02_l, bone2Name=dyn_pigtailA_01_l, swing1LimitDegrees=22, swing2LimitDegrees=22, twistLimitDegrees=12}
 {bone1Name=dyn_pigtailA_03_l, bone2Name=dyn_pigtailA_02_l, swing1LimitDegrees=32, swing2LimitDegrees=32, twistLimitDegrees=12}
 {bone1Name=dyn_pigtailA_04_l, bone2Name=dyn_pigtailA_03_l, swing1LimitDegrees=32, swing2LimitDegrees=32, twistLimitDegrees=12}
 {bone1Name=dyn_pigtailA_05_l, bone2Name=dyn_pigtailA_04_l, swing1LimitDegrees=32, swing2LimitDegrees=32, twistLimitDegrees=12}
 {
  bone1Name=clavicle_l
  bone2Name=spine_05
  swing1LimitDegrees=45
  swing1Motion=Free
  swing2LimitDegrees=45
  swing2Motion=Free
  twistLimitDegrees=45
  twistMotion=Free
 }
 {
  bone1Name=clavicle_r
  bone2Name=spine_05
  swing1LimitDegrees=45
  swing1Motion=Free
  swing2LimitDegrees=45
  swing2Motion=Free
  twistLimitDegrees=45
  twistMotion=Free
 }
 {
  bone1Name=head
  bone2Name=neck_01
  swing1LimitDegrees=45
  swing1Motion=Free
  swing2LimitDegrees=45
  swing2Motion=Free
  twistLimitDegrees=45
  twistMotion=Free
 }
 {
  bone1Name=upperarm_r
  bone2Name=clavicle_r
  swing1LimitDegrees=45
  swing1Motion=Free
  swing2LimitDegrees=45
  swing2Motion=Free
  twistLimitDegrees=45
  twistMotion=Free
 }
 {
  bone1Name=upperarm_l
  bone2Name=clavicle_l
  swing1LimitDegrees=45
  swing1Motion=Free
  swing2LimitDegrees=45
  swing2Motion=Free
  twistLimitDegrees=45
  twistMotion=Free
 }
 {
  bone1Name=neck_01
  bone2Name=spine_05
  swing1LimitDegrees=45
  swing1Motion=Free
  swing2LimitDegrees=45
  swing2Motion=Free
  twistLimitDegrees=45
  twistMotion=Free
 }
]
